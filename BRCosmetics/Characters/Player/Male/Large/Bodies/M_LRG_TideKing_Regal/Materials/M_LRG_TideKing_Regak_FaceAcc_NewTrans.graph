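# Translucent character material, named for the face accessory of the large male TideKing Regal body, built on the shared 2023 character parent functions and flagged for skeletal mesh use.
# Between the PreFX and PostFX functions it multiplies opacity by a binary depth test, leaving other material attributes intact.
# Pixels whose depth plus the DepthBias parameter exceeds scene custom depth become fully transparent, and translucent custom depth writes are enabled.
# A second If node is left unconnected.
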Material /BRCosmetics/Characters/Player/Male/Large/Bodies/M_LRG_TideKing_Regal/Materials/M_LRG_TideKing_Regak_FaceAcc_NewTrans
graph {
 properties {
  allowTranslucentCustomDepthWrites=true
  bUseMaterialAttributes=true
  bUsedWithMeshDeformer=true
  bUsedWithMorphTargets=true
  bUsedWithNanite=true
  bUsedWithSkeletalMesh=true
  blendMode=BLEND_Translucent
  pixelDepthOffsetMode=PDOM_Legacy
  shadingModel=MSM_FromMaterialExpression
  translucencyLightingMode=TLM_SurfacePerPixelLighting
  usedShadingModels=Default Lit | Subsurface | Clear Coat | Subsurface Profile | Cloth
 }
 nodes [
  n1 = Constant()
  If(B=Constant() {r=100}, "A > B"=Constant() {r=1}, "A == B"=@n1, "A < B"=@n1)
  n2 = MaterialFunctionCall() {materialFunction=&MF_CharacterParent_2023_PreFX}
  n3 = GetMaterialAttributes(None=@n2.Result) {attributeGetTypes [#1]}
  n4 = Constant() {r=1}
  n5 = If(
   A=Add(A=PixelDepth().R, B=ScalarParameter() {defaultValue=-1, group=AdvancedTrans, parameterName=DepthBias})
   B=ComponentMask(None=SceneTexture() {sceneTextureId=PPI_CustomDepth}.Color) {r=true}
   "A > B"=Constant()
   "A == B"=@n4
   "A < B"=@n4
  )
  MaterialFunctionCall(
   Material=SetMaterialAttributes!(MaterialAttributes=@n3.MaterialAttributes, Opacity=Multiply(A=@n3.Opacity, B=@n5)) {attributeSetTypes [#1]}
   M=@n2.M
   Pre Skinned Local Position=@n2.M
   Distance Blend=@n2.M
   Fresnel=@n2.M
  ) {materialFunction=&MF_CharacterParent_2023_PostFX}
 ]
}
