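# Character master material for the TideKing Regal skin, adding an optional slurp effect to the 2023 character parent shading.
# The PreFX parent function supplies base attributes, and channel-selectable Tech Art Mask components drive slurp alpha, bubbles and normal distortion in MF_TideKing_Slurp.
# The UseSlurp static switch selects the slurp result or the untouched base, then feeds the PostFX parent function.
# It uses material attributes with an expression-driven shading model.
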Material /BRCosmetics/Characters/Player/Male/Large/Bodies/M_LRG_TideKing_Regal/Materials/M_TideKing_Regal_Slurp
graph {
 properties {
  bUseMaterialAttributes=true
  bUsedWithClothing=true
  bUsedWithMeshDeformer=true
  bUsedWithMorphTargets=true
  bUsedWithNanite=true
  bUsedWithNiagaraMeshParticles=true
  bUsedWithSkeletalMesh=true
  pixelDepthOffsetMode=PDOM_Legacy
  shadingModel=MSM_FromMaterialExpression
  usedShadingModels=Default Lit | Subsurface | Clear Coat | Subsurface Profile | Cloth
 }
 nodes [
  n1 = MaterialFunctionCall() {materialFunction=&MF_CharacterParent_2023_PreFX}
  n2 = StaticComponentMaskParameter(None=@n1.Tech Art Mask) {defaultR=true, group=Skin FX, parameterName=Slurp Channel}
  n3 = ComponentMask(None=@n1.M) {r=true}
  n4 = StaticComponentMaskParameter(None=@n1.Tech Art Mask) {defaultG=true, group=Skin FX, parameterName=Slurp Bubbles Channel}
  n5 = StaticComponentMaskParameter(None=@n1.Tech Art Mask) {defaultB=true, group=Skin FX, parameterName=Slurp Normal Distortion Channel}
  MaterialFunctionCall(
   Material=StaticSwitchParameter(
    True=MaterialFunctionCall!(
     Alpha=@n2
     AO=@n3
     BubblesMask=@n4
     SlurpNormalDistortionMask=@n5
     Pre Skinned Position=@n1.Pre skinned Local Position
     Material=@n1.Pre skinned Local Position
    ) {materialFunction=/Game/Characters/Player/Male/Large/Bodies/M_LRG_TideKing_Regal/Materials/MF_TideKing_Slurp}.Result
    False=@n1.Result
   ) {group=Skin FX, parameterName=UseSlurp}
   M=@n1.M
   Pre Skinned Local Position=@n1.M
   Distance Blend=@n1.M
   Fresnel=@n1.M
  ) {materialFunction=&MF_CharacterParent_2023_PostFX}
 ]
}
