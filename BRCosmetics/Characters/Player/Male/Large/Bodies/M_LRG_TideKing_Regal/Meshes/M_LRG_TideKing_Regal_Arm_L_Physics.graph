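# Physics asset for the left arm mesh of the large male TideKing Regal character cosmetic, simulating small dangling "blob" bones at the shoulder, elbow and forearm.
# Each site has a three-capsule chain, from a master bone through 01 to 02, joined by constraints with tight, uniform swing and twist limits.
# A large unconstrained spine_05 capsule stands in for the upper torso, likely as a collider for the chains.
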PhysicsAsset /BRCosmetics/Characters/Player/Male/Large/Bodies/M_LRG_TideKing_Regal/Meshes/M_LRG_TideKing_Regal_Arm_L_Physics
bodies {
 dyn_elb_blob_01_l [{shapeType=Capsule, length=1.7033, radius=2.1641, shapeName=dyn_elb_blob_01_r_capsule, center=<2.495 0 0>, rotation=<rot -90 0 0>}]
 dyn_elb_blob_02_l [{shapeType=Capsule, length=1.2372, radius=1.5863, shapeName=dyn_elb_blob_02_r_capsule, center=<1.1866 0 0>, rotation=<rot -90 0 0>}]
 dyn_elb_blob_mstr_l [{shapeType=Capsule, length=0.505, radius=0.505, shapeName=dyn_elb_blob_mstr_r_capsule, rotation=<rot 0 -90 0>}]
 dyn_forearm_blob_01_l [{shapeType=Capsule, length=1.5907, radius=1.9467, shapeName=dyn_forearm_blob_01_r_capsule, center=<2.1144 0 0>, rotation=<rot -90 0 0>}]
 dyn_forearm_blob_02_l [{shapeType=Capsule, length=1.2372, radius=1.4478, shapeName=dyn_forearm_blob_02_r_capsule, center=<1.4244 0 0>, rotation=<rot -90 0 0>}]
 dyn_forearm_blob_mstr_l [{shapeType=Capsule, length=0.505, radius=0.505, shapeName=dyn_forearm_blob_mstr_r_capsule, rotation=<rot 0 -90 0>}]
 dyn_shld_blob_01_l [{shapeType=Capsule, length=2.3474, radius=2.4656, shapeName=dyn_shld_blob_01_r_capsule, center=<2.2435 0 0>, rotation=<rot -90 0 0>}]
 dyn_shld_blob_02_l [{shapeType=Capsule, length=1.1735, radius=1.7249, shapeName=dyn_shld_blob_02_r_capsule, center=<0.6732 0 0>, rotation=<rot -90 180 180>}]
 dyn_shld_blob_mstr_l [{shapeType=Capsule, length=0.505, radius=0.505, shapeName=dyn_shld_blob_mstr_r_capsule, rotation=<rot 0 -90 0>}]
 spine_05 [{
   shapeType=Capsule
   length=16.1302
   radius=15.9525
   shapeName=spine_05_capsule
   center=<-1.3348 -2.4068 0.0277>
   rotation=<rot -0.366 -90 0.1882>
  }]
}
constraints [
 {bone1Name=dyn_elb_blob_02_l, bone2Name=dyn_elb_blob_01_l, swing1LimitDegrees=2, swing2LimitDegrees=2, twistLimitDegrees=2}
 {bone1Name=dyn_forearm_blob_02_l, bone2Name=dyn_forearm_blob_01_l, swing1LimitDegrees=2, swing2LimitDegrees=2, twistLimitDegrees=2}
 {bone1Name=dyn_forearm_blob_01_l, bone2Name=dyn_forearm_blob_mstr_l, swing1LimitDegrees=2, swing2LimitDegrees=2, twistLimitDegrees=2}
 {bone1Name=dyn_elb_blob_01_l, bone2Name=dyn_elb_blob_mstr_l, swing1LimitDegrees=2, swing2LimitDegrees=2, twistLimitDegrees=2}
 {bone1Name=dyn_shld_blob_01_l, bone2Name=dyn_shld_blob_mstr_l, swing1LimitDegrees=2, swing2LimitDegrees=2, twistLimitDegrees=2}
 {bone1Name=dyn_shld_blob_02_l, bone2Name=dyn_shld_blob_01_l, swing1LimitDegrees=2, swing2LimitDegrees=2, twistLimitDegrees=2}
]
