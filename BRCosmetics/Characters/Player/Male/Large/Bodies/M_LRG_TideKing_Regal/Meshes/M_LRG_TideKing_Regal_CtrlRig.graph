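# Control Rig for the large male TideKing Regal body that corrects deformation and armor, shell and buckle placement.
# A sequence runs the shared DeformRig function, then applies constant offsets to armor and shell bones.
# Spherical pose readers on thighs and spine feed interpolated variables that weight further corrective bone transforms, including right calf blob.
# Finally, conch eye bones get rotational offsets and ChainHarmonics wobble with randomized speeds.
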ControlRigBlueprint /BRCosmetics/Characters/Player/Male/Large/Bodies/M_LRG_TideKing_Regal/Meshes/M_LRG_TideKing_Regal_CtrlRig
graphs [
 RigVMFunctionLibrary {links [], nodes []}
 RigVMFunction_Sequence_1_ContainedGraph {nodes [
   n1 = RigVMFunctionEntryNode()
   RigVMFunction_Sequence = &RigVMFunction_Sequence(ExecuteContext=@n1.ExecuteContext)
   n2 = &RigVMFunction_Sequence(ExecuteContext=@RigVMFunction_Sequence.B)
   n3 = &RigVMFunction_Sequence(ExecuteContext=@n2.B)
   n4 = &RigVMFunction_Sequence(ExecuteContext=@n3.B)
   RigVMFunctionReturnNode(A=@RigVMFunction_Sequence.A, B=@n2.A, C=@n3.A, D=@n4.A, E=@n4.B)
  ]}
 RigVMModel {nodes [
   RigUnit_BeginExecution = &RigUnit_BeginExecution()
   RigVMFunction_Sequence = RigVMAggregateNode(ExecuteContext=@RigUnit_BeginExecution.ExecutePin)
   FN_DeformRig_FNC = RigVMFunctionReferenceNode(
    DeformRig_ Curve="(deformRig_l_upperarm_up_alpha_72_#1=False,deformRig_r_upperarm_up_alpha_73_#2=False,deformRig_l_upperarm_fr_alpha_74_#3=False,deformRig_r_upperarm_fr_alpha_75_CB51B8 ...(truncated)"
    DeformRig_Bool="(upperarm_1_#4=True,elbow_3_#5=True,wrist_7_#6=True,groin_8_#7=True,glute_17_#8=True,knee_9_40C408324CA261 ...(truncated)"
    DeformRig_Debug=(upperarm_debug_18_#4=False,wrist_debug_20_#6=False,lat_debug_26_#9=False,groinGlute_debug_27_#7=False)
    DeformRig_Value="(upperarm_up_18_#10=(Rotation=(X=0,Y=0,Z=0,W=1),Translation=(X=0,Y=0,Z=8),Scale3D=(X=0.7,Y=1,Z=1)),upperarm_fwd_fr_19_#11=(Rotation=(X=0,Y=0,Z=0,W=1),Translation=(X=0,Y=10,Z=0),Scale3D=(X=0.5,Y=1,Z= ...(truncated)"
    ExecuteContext=@RigVMFunction_Sequence.A
   )
   &RigUnit_ModifyTransforms(
    ItemToModify=((Item=(Type=Bone,Name="dyn_armor_mstr_l"),Transform=(Rotation=(X=0,Y=0,Z=0.033,W=0.9995),Translation=(X=0.7,Y=-1.6569,Z=0),Scale3D=(X=1,Y=1,Z=1))))
    Weight=1
    WeightMinimum=0
    ExecutePin=@RigVMFunction_Sequence.B
   )
   > &RigUnit_ModifyTransforms(
    ItemToModify=((Item=(Type=Bone,Name="dyn_armor_mstr_r"),Transform=(Rotation=(X=-0.0374,Y=-0.0468,Z=0.0257,W=0.9979),Translation=(X=-0.7598,Y=1.2692,Z=-0.08),Scale3D=(X=1,Y=1,Z=1))))
    Weight=1
    WeightMinimum=0
   )
   > &RigUnit_ModifyTransforms(
    ItemToModify=((Item=(Type=Bone,Name="dyn_shell_01_mstr_r"),Transform=(Rotation=(X=0.088,Y=0.0458,Z=-0.0314,W=0.9946),Translation=(X=-0.9999,Y=0.1551,Z=0.9352),Scale3D=(X=1,Y=1,Z=1))))
    Weight=1
    WeightMinimum=0
   )
   > &RigUnit_ModifyTransforms(
    ItemToModify=((Item=(Type=Bone,Name="dyn_shell_02_mstr_r"),Transform=(Rotation=(X=0.0435,Y=0.0098,Z=-0.0487,W=0.9978),Translation=(X=0.1603,Y=-0.3308,Z=1.2389),Scale3D=(X=1,Y=1,Z=1))))
    Weight=1
    WeightMinimum=0
   )
   > &RigUnit_ModifyTransforms(
    ItemToModify=((Item=(Type=Bone,Name="dyn_shell_01_mstr_l"),Transform=(Rotation=(X=0,Y=0,Z=0,W=1),Translation=(X=-1.51,Y=-0.1634,Z=0),Scale3D=(X=1,Y=1,Z=1))))
    Weight=1
    WeightMinimum=0
   )
   > &RigUnit_ModifyTransforms(
    ItemToModify=((Item=(Type=Bone,Name="dyn_shell_02_mstr_l"),Transform=(Rotation=(X=0,Y=0,Z=0,W=1),Translation=(X=-0.7088,Y=0,Z=-0.5875),Scale3D=(X=1,Y=1,Z=1))))
    Weight=1
    WeightMinimum=0
   )
   n1 = &RigUnit_SphericalPoseReader(
    ActiveRegionScaleFactors=(PositiveWidth=1,NegativeWidth=1,PositiveHeight=1,NegativeHeight=1)
    ActiveRegionSize=0.1
    Debug=(bDrawDebug=false,bDraw2D=False,bDrawLocalAxes=False,DebugScale=25,DebugSegments=20,DebugThickness=0.25)
    DriverAxis=(X=1,Y=0,Z=0)
    DriverItem=(Type=Bone,Name="thigh_r")
    FalloffRegionScaleFactors=(PositiveWidth=1,NegativeWidth=1,PositiveHeight=1,NegativeHeight=1)
    FalloffSize=0.5
    FlipHeightScaling=False
    FlipWidthScaling=False
    OptionalParentItem=(Type=Bone,Name="None")
    RotationOffset=(X=-91.1322,Y=-5.7835,Z=24.4661)
    ExecutePin=@RigVMFunction_Sequence.C
   )
   n2 = &RigVMFunction_AlphaInterp(InterpSpeedDecreasing=4, InterpSpeedIncreasing=4, Scale=1, bClampResult=False, bInterpResult=True, Value=@n1.OutputParam)
   > /Script/ControlRig.RigUnit_SetCurveValue(Curve=SPR_Thigh_R, Value=@n2.Result) {is=struct RigVMUnitNode}
   > RigVMVariableNode(Value=@n2.Result) {variable=Thigh_R_Adjust}
   > n3 = &RigUnit_SphericalPoseReader(
    ActiveRegionScaleFactors=(PositiveWidth=1,NegativeWidth=1,PositiveHeight=1,NegativeHeight=1)
    ActiveRegionSize=0.05
    Debug=(bDrawDebug=false,bDraw2D=False,bDrawLocalAxes=False,DebugScale=25,DebugSegments=20,DebugThickness=0.25)
    DriverAxis=(X=1,Y=0,Z=0)
    DriverItem=(Type=Bone,Name="thigh_r")
    FalloffRegionScaleFactors=(PositiveWidth=0.5,NegativeWidth=0.5,PositiveHeight=1,NegativeHeight=1)
    FalloffSize=0.4
    FlipHeightScaling=False
    FlipWidthScaling=False
    OptionalParentItem=(Type=Bone,Name="None")
    RotationOffset=(X=-107.3487,Y=-5.7835,Z=8.8095)
   )
   n4 = &RigVMFunction_AlphaInterp(InterpSpeedDecreasing=14, InterpSpeedIncreasing=14, Scale=1, bClampResult=False, bInterpResult=True, Value=@n3.OutputParam)
   > RigVMVariableNode(Value=@n4.Result) {variable=Thigh_R_Strafe_Adjust}
   > n5 = &RigUnit_SphericalPoseReader(
    ActiveRegionScaleFactors=(PositiveWidth=1,NegativeWidth=1,PositiveHeight=1,NegativeHeight=1)
    ActiveRegionSize=0.01
    Debug=(bDrawDebug=false,bDraw2D=False,bDrawLocalAxes=False,DebugScale=25,DebugSegments=20,DebugThickness=0.25)
    DriverAxis=(X=-1,Y=0,Z=0)
    DriverItem=(Type=Bone,Name="thigh_l")
    FalloffRegionScaleFactors=(PositiveWidth=1,NegativeWidth=1,PositiveHeight=1,NegativeHeight=1)
    FalloffSize=0.45
    FlipHeightScaling=False
    FlipWidthScaling=False
    OptionalParentItem=(Type=Bone,Name="None")
    RotationOffset=(X=110.2686,Y=0,Z=10.5195)
   )
   RigVMFunction_AlphaInterp = &RigVMFunction_AlphaInterp(InterpSpeedDecreasing=4, InterpSpeedIncreasing=4, Scale=1, bClampResult=False, bInterpResult=True, Value=@n5.OutputParam)
   > RigVMVariableNode(Value=@RigVMFunction_AlphaInterp.Result) {variable=Thigh_L_Adjust}
   > n6 = &RigUnit_SphericalPoseReader(
    ActiveRegionScaleFactors=(PositiveWidth=1,NegativeWidth=1,PositiveHeight=1,NegativeHeight=1)
    ActiveRegionSize=0.1
    Debug=(bDrawDebug=false,bDraw2D=False,bDrawLocalAxes=False,DebugScale=25,DebugSegments=20,DebugThickness=0.25)
    DriverAxis=(X=-1,Y=0,Z=0)
    DriverItem=(Type=Bone,Name="thigh_l")
    FalloffRegionScaleFactors=(PositiveWidth=1,NegativeWidth=1,PositiveHeight=1,NegativeHeight=1)
    FalloffSize=0.4
    FlipHeightScaling=False
    FlipWidthScaling=False
    OptionalParentItem=(Type=Bone,Name="None")
    RotationOffset=(X=83.6929,Y=0,Z=4.8928)
   )
   n7 = &RigVMFunction_AlphaInterp(InterpSpeedDecreasing=14, InterpSpeedIncreasing=14, Scale=1, bClampResult=False, bInterpResult=True, Value=@n6.OutputParam)
   > RigVMVariableNode(Value=@n7.Result) {variable=Thigh_L_Strafe_Adjust}
   > n8 = &RigUnit_SphericalPoseReader(
    ActiveRegionScaleFactors=(PositiveWidth=0.7,NegativeWidth=0.7,PositiveHeight=1,NegativeHeight=1)
    ActiveRegionSize=0.05
    Debug=(bDrawDebug=true,bDraw2D=False,bDrawLocalAxes=False,DebugScale=25,DebugSegments=20,DebugThickness=0.25)
    DriverAxis=(X=0,Y=1,Z=0)
    DriverItem=(Type=Bone,Name="spine_05")
    FalloffRegionScaleFactors=(PositiveWidth=1,NegativeWidth=0.5,PositiveHeight=1,NegativeHeight=1)
    FalloffSize=0.3
    FlipHeightScaling=False
    FlipWidthScaling=False
    OptionalParentItem=(Type=Bone,Name="pelvis")
    RotationOffset=(X=82.2057,Y=0,Z=30.7285)
   )
   n9 = &RigVMFunction_AlphaInterp(InterpSpeedDecreasing=14, InterpSpeedIncreasing=14, Scale=1, bClampResult=False, bInterpResult=True, Value=@n8.OutputParam)
   > RigVMVariableNode(Value=@n9.Result) {variable=Spine_Adjust}
   n10 = RigVMVariableNode() {variable=Thigh_R_Adjust}
   &RigUnit_ModifyTransforms(
    ItemToModify=((Item=(Type=Bone,Name="dyn_armor_mstr_r"),Transform=(Rotation=(X=-0.0056,Y=0.0325,Z=-0.0276,W=0.9991),Translation=(X=-9.7249,Y=-0.5829,Z=-0.5049),Scale3D=(X=0.8973,Y=1,Z=1))))
    WeightMinimum=0
    ExecutePin=@RigVMFunction_Sequence.D
    Weight=@n10.Value
   )
   Add = &RigVMFunction_MathDoubleAdd(B=0.25, A=@n10.Value)
   > &RigUnit_ModifyTransforms(
    ItemToModify=((Item=(Type=Bone,Name="dyn_shell_01_mstr_r"),Transform=(Rotation=(X=0.0604,Y=-0.0159,Z=-0.1165,W=0.9912),Translation=(X=1.3775,Y=0.3172,Z=2.6048),Scale3D=(X=1,Y=1,Z=1))))
    WeightMinimum=0
    Weight=@Add.Result
   )
   > &RigUnit_ModifyTransforms(
    ItemToModify=((Item=(Type=Bone,Name="dyn_shell_02_mstr_r"),Transform=(Rotation=(X=0.1395,Y=0.0011,Z=-0.015,W=0.9901),Translation=(X=-0.6846,Y=0.9747,Z=1.6283),Scale3D=(X=1,Y=1,Z=1))))
    WeightMinimum=0
    Weight=@n10.Value
   )
   > &RigUnit_ModifyTransforms(
    ItemToModify=((Item=(Type=Bone,Name="dyn_blob_calf_mstr_r"),Transform=(Rotation=(X=0,Y=0,Z=0,W=1),Translation=(X=2.4048,Y=0,Z=0.0476),Scale3D=(X=1,Y=1,Z=1))))
    WeightMinimum=0
    Weight=@n10.Value
   )
   n11 = RigVMVariableNode() {variable=Thigh_R_Strafe_Adjust}
   > &RigUnit_ModifyTransforms(
    ItemToModify=((Item=(Type=Bone,Name="dyn_armor_fr_mstr"),Transform=(Rotation=(X=0.5499,Y=-0.1647,Z=-0.3018,W=0.7612),Translation=(X=0,Y=0,Z=0),Scale3D=(X=1,Y=1,Z=1))))
    WeightMinimum=0
    Weight=@n11.Value
   )
   > &RigUnit_ModifyTransforms(
    ItemToModify=((Item=(Type=Bone,Name="dyn_thigh_adjust_r"),Transform=(Rotation=(X=0,Y=0,Z=0,W=1),Translation=(X=-4.1317,Y=-4.2354,Z=0),Scale3D=(X=1,Y=1,Z=1))))
    WeightMinimum=0
    Weight=@n11.Value
   )
   n12 = RigVMVariableNode() {variable=Thigh_L_Adjust}
   > &RigUnit_ModifyTransforms(
    ItemToModify=((Item=(Type=Bone,Name="dyn_armor_mstr_l"),Transform=(Rotation=(X=0.028,Y=-0.0611,Z=-0.0214,W=0.9975),Translation=(X=11.0716,Y=1.5519,Z=-2.5377),Scale3D=(X=0.8973,Y=1,Z=1))))
    WeightMinimum=0
    Weight=@n12.Value
   )
   > &RigUnit_ModifyTransforms(
    ItemToModify=((Item=(Type=Bone,Name="dyn_shell_01_mstr_l"),Transform=(Rotation=(X=0.1267,Y=0.038,Z=-0.1774,W=0.9752),Translation=(X=1.6241,Y=-0.5441,Z=-2.8749),Scale3D=(X=1,Y=1,Z=1))))
    WeightMinimum=0
    Weight=&RigVMFunction_MathDoubleAdd(B=0.4, A=@n12.Value).Result
   )
   > &RigUnit_ModifyTransforms(
    ItemToModify=((Item=(Type=Bone,Name="dyn_shell_02_mstr_l"),Transform=(Rotation=(X=0.0878,Y=-0.0469,Z=0.0043,W=0.995),Translation=(X=1.2488,Y=0.5626,Z=-1.6267),Scale3D=(X=1,Y=1,Z=1))))
    WeightMinimum=0
    Weight=@n12.Value
   )
   n13 = RigVMVariableNode() {variable=Thigh_L_Strafe_Adjust}
   > &RigUnit_ModifyTransforms(
    ItemToModify=((Item=(Type=Bone,Name="dyn_armor_fr_mstr"),Transform=(Rotation=(X=-0.4283,Y=-0.011,Z=-0.3139,W=0.8473),Translation=(X=0.1416,Y=0.6795,Z=0),Scale3D=(X=1,Y=1,Z=1))))
    WeightMinimum=0
    Weight=@n13.Value
   )
   > &RigUnit_ModifyTransforms(
    ItemToModify=((Item=(Type=Bone,Name="dyn_thigh_adjust_l"),Transform=(Rotation=(X=0,Y=0,Z=0,W=1),Translation=(X=-0.9632,Y=5.1468,Z=0),Scale3D=(X=1,Y=1,Z=1))))
    WeightMinimum=0
    Weight=@n13.Value
   )
   > &RigUnit_ModifyTransforms(
    ItemToModify=((Item=(Type=Bone,Name="dyn_buckle_mstr"),Transform=(Rotation=(X=0.0238,Y=-0.0545,Z=0.1678,W=0.984),Translation=(X=0.9142,Y=-1.0871,Z=0),Scale3D=(X=1,Y=1,Z=1))))
    WeightMinimum=0
    Weight=RigVMVariableNode() {variable=Spine_Adjust}.Value
   )
   &RigUnit_ModifyTransforms(
    ItemToModify=((Item=(Type=Bone,Name="dyn_conch_eye_harmonics_mstr_l"),Transform=(Rotation=(X=0.0515,Y=0.1743,Z=0.0075,W=0.9833),Translation=(X=0,Y=0,Z=0),Scale3D=(X=1,Y=1,Z=1))))
    Weight=1
    WeightMinimum=0
    ExecutePin=@RigVMFunction_Sequence.E
   )
   > &RigUnit_ModifyTransforms(
    ItemToModify=((Item=(Type=Bone,Name="dyn_conch_eye_harmonics_mstr_r"),Transform=(Rotation=(X=-0.003,Y=0.1736,Z=0.0172,W=0.9847),Translation=(X=0,Y=0,Z=0),Scale3D=(X=1,Y=1,Z=1))))
    Weight=1
    WeightMinimum=0
   )
   Random = /Script/RigVM.RigVMFunction_RandomFloat(
    Maximum=2
    Minimum=0
    Seed=150
    Duration=&RigVMFunction_RandomFloat(Duration=1, Maximum=2, Minimum=0, Seed=200) {template="Random::Execute(in Seed,in Minimum,in Maximum,in Duration,out Result)"}.Result
   ) {is=struct RigVMUnitNode, template="Random::Execute(in Seed,in Minimum,in Maximum,in Duration,out Result)"}
   n14 = &RigVMFunction_RandomFloat(
    Maximum=3
    Minimum=0
    Seed=217
    Duration=&RigVMFunction_RandomFloat(Duration=1, Maximum=4, Minimum=2, Seed=200) {template="Random::Execute(in Seed,in Minimum,in Maximum,in Duration,out Result)"}.Result
   ) {template="Random::Execute(in Seed,in Minimum,in Maximum,in Duration,out Result)"}
   > /Script/ControlRig.RigUnit_ChainHarmonicsPerItem(
    ChainRoot=(Type=Bone,Name="dyn_conch_eye_harmonics_l")
    DrawWorldOffset=(Rotation=(X=0,Y=0,Z=0,W=1),Translation=(X=0,Y=0,Z=0),Scale3D=(X=1,Y=1,Z=1))
    Pendulum=(bEnabled=true,PendulumStiffness=3,PendulumGravity=(X=0,Y=0,Z=0),PendulumBlend=1,PendulumDrag=0.98,PendulumMaximum=0.1,UnwindAxis=(X=0,Y=1,Z=0),UnwindMinimum=0.2,UnwindMaximum=0.05)
    Reach=(bEnabled=True,ReachTarget=(X=0,Y=0,Z=0),ReachAxis=(X=1,Y=0,Z=0))
    Speed=(X=0,Y=0,Z=2)
    Wave=(bEnabled=True,WaveFrequency=(X=0,Y=1,Z=0.5),WaveAmplitude=(X=0,Y=0.3,Z=0.3),WaveOffset=(X=0,Y=0,Z=0),WaveNoise=(X=0.5,Y=0.5,Z=0.5),WaveMinimum=0,WaveMaximum=3,WaveEase=SineEaseInOut)
    WaveCurve=(EditorCurveData=(Keys=((),(Time=1,Value=1)),DefaultValue=340282346638528859811704183484516925440,PreInfinityExtrap=RCCE_Constant,PostInfinityExtrap=RCCE_Constant))
    bDrawDebug=true
    Speed.Z=@Random.Result
    Speed.Y=@n14.Result
   ) {is=struct RigVMUnitNode}
   > &RigUnit_ChainHarmonicsPerItem(
    ChainRoot=(Type=Bone,Name="dyn_conch_eye_harmonics_r")
    DrawWorldOffset=(Rotation=(X=0,Y=0,Z=0,W=1),Translation=(X=0,Y=0,Z=0),Scale3D=(X=1,Y=1,Z=1))
    Pendulum=(bEnabled=true,PendulumStiffness=3,PendulumGravity=(X=0,Y=0,Z=0),PendulumBlend=1,PendulumDrag=0.98,PendulumMaximum=0.1,UnwindAxis=(X=0,Y=1,Z=0),UnwindMinimum=0.2,UnwindMaximum=0.05)
    Reach=(bEnabled=True,ReachTarget=(X=0,Y=0,Z=0),ReachAxis=(X=1,Y=0,Z=0))
    Speed=(X=0,Y=0,Z=2)
    Wave=(bEnabled=True,WaveFrequency=(X=1,Y=1,Z=0.5),WaveAmplitude=(X=0,Y=0.3,Z=0.3),WaveOffset=(X=0,Y=0,Z=0),WaveNoise=(X=0.5,Y=0.5,Z=0.5),WaveMinimum=1,WaveMaximum=3,WaveEase=SineEaseInOut)
    WaveCurve=(EditorCurveData=(Keys=((),(Time=1,Value=1)),DefaultValue=340282346638528859811704183484516925440,PreInfinityExtrap=RCCE_Constant,PostInfinityExtrap=RCCE_Constant))
    bDrawDebug=true
    Speed.Z=@Random.Result
    Speed.Y=@n14.Result
   )
  ]}
]
hierarchy [
 root {loc=<0 0 0>, rot=<0 0 0 1>}
 pelvis {parent=root, loc=<0 -0.7321 92.4218>, rot=<0.0127 0.707 -0.0127 -0.707>}
 spine_01 {parent=pelvis, loc=<5.6002 -0.5128 0>, rot=<0 0 0.034 -0.9994>}
 spine_02 {parent=spine_01, loc=<8.0753 0.0139 0>, rot=<0 0 0.0238 -0.9997>}
 spine_03 {parent=spine_02, loc=<7.0393 0 0>, rot=<0 0 0.0149 -0.9999>}
 spine_04 {parent=spine_03, loc=<8.1024 0 0>, rot=<0 0 -0.0132 -0.9999>}
 spine_05 {parent=spine_04, loc=<10.0146 0 0>, rot=<0 0 -0.0015 -1>}
 attach_cape {parent=spine_05, loc=<8.9965 -11.6913 -0.0085>, rot=<0.0616 -0.7044 -0.0616 -0.7044>}
 attach_backpack {parent=spine_05, loc=<-0.1234 -13.4449 -0.0076>, rot=<-0.0093 -0.707 0.0093 -0.707>}
 neck_01 {parent=spine_05, loc=<10.815 1.9196 0>, rot=<0 0 -0.1864 -0.9825>}
 neck_02 {parent=neck_01, loc=<4.7347 0.3045 0>, rot=<0 0 0.0363 -0.9993>}
 head {parent=neck_02, loc=<5.7632 0.5552 0>, rot=<0 0 0.115 -0.9934>}
 clavicle_l {parent=spine_05, loc=<2.5856 -0.2866 -4.3777>, rot=<0.0465 -0.7289 -0.009 -0.6829>}
 upperarm_l {parent=clavicle_l, loc=<18.2264 0.1342 -0.001>, rot=<0.0253 -0.2693 0.0013 -0.9627>}
 lowerarm_l {parent=upperarm_l, loc=<29.0071 0.1686 -1.239>, rot=<0.0391 -0.0099 -0.2453 -0.9686>}
 hand_l {parent=lowerarm_l, loc=<25.8607 0.222 1.0274>, rot=<-0.7193 -0.0988 0.1193 -0.6773>}
 thumb_01_l {parent=hand_l, loc=<1.3118 -0.3191 -2.459>, rot=<0.6546 -0.4042 0.0664 -0.6354>}
 thumb_02_l {parent=thumb_01_l, loc=<5.9336 -0.0448 0.0401>, rot=<-0.0037 0.0213 0.298 -0.9543>}
 thumb_03_l {parent=thumb_02_l, loc=<2.2223 0.3682 -0.0907>, rot=<0.0118 -0.013 0.0066 -0.9998>}
 index_metacarpal_l {parent=hand_l, loc=<1.6535 0.0183 -2.96>, rot=<0.0158 -0.0124 -0.0173 -0.9996>}
 index_01_l {parent=index_metacarpal_l, loc=<8.3008 0.0357 -0.0524>, rot=<-0.0044 0.0485 0.0924 -0.9945>}
 index_02_l {parent=index_01_l, loc=<3.3587 -0.3004 0.0452>, rot=<-0.002 0.0027 0.3225 -0.9466>}
 index_03_l {parent=index_02_l, loc=<2.5578 -0.123 -0.0252>, rot=<-0.002 0.0067 0.2558 -0.9667>}
 middle_metacarpal_l {parent=hand_l, loc=<2.1856 -0.1725 -0.5984>, rot=<0.006 0.0168 -0.0329 -0.9993>}
 middle_01_l {parent=middle_metacarpal_l, loc=<7.3275 -0.2447 0.0129>, rot=<-0.0208 0.0042 0.0858 -0.9961>}
 middle_02_l {parent=middle_01_l, loc=<4.0123 0.1401 0.1175>, rot=<-0.0218 0.0258 0.3512 -0.9357>}
 middle_03_l {parent=middle_02_l, loc=<2.6077 0.0534 0.1373>, rot=<0.0082 0.0188 0.1708 -0.9851>}
 ring_metacarpal_l {parent=hand_l, loc=<2.2074 -0.3625 1.3638>, rot=<-0.0284 0.0722 -0.0458 -0.9959>}
 ring_01_l {parent=ring_metacarpal_l, loc=<7.0295 -0.6197 -0.1446>, rot=<-0.0489 -0.0252 0.1892 -0.9804>}
 ring_02_l {parent=ring_01_l, loc=<3.6174 0.4647 -0.1027>, rot=<0 0 0.2806 -0.9598>}
 ring_03_l {parent=ring_02_l, loc=<2.4152 0.114 0.0199>, rot=<0 0 0.1189 -0.9929>}
 pinky_metacarpal_l {parent=hand_l, loc=<1.7616 -0.9753 3.3814>, rot=<-0.0397 0.0814 -0.069 -0.9935>}
 pinky_01_l {parent=pinky_metacarpal_l, loc=<7.1709 -0.4831 -0.1094>, rot=<-0.0654 -0.0035 0.2242 -0.9723>}
 pinky_02_l {parent=pinky_01_l, loc=<2.8274 0.0893 -0.1043>, rot=<0.0037 0.0011 0.2854 -0.9584>}
 pinky_03_l {parent=pinky_02_l, loc=<2.2046 0.0307 -0.0078>, rot=<-0.0002 0.0098 0.1445 -0.9895>}
 deform_wrist_dn_base_l {parent=hand_l, loc=<0 -3.4173 0>, rot=<0 0 0 -1>}
 deform_wrist_dn_l {parent=deform_wrist_dn_base_l, loc=<0 0 0>, rot=<0 0 0 1>}
 deform_wrist_up_base_l {parent=hand_l, loc=<0 2.7805 0>, rot=<0 0 0 -1>}
 deform_wrist_up_l {parent=deform_wrist_up_base_l, loc=<0 0 0>, rot=<0 0 0 1>}
 lowerarm_twist_02_l {parent=lowerarm_l, loc=<9.6199 0 0>, rot=<0 0 0 -1>}
 lowerarm_twist_01_l {parent=lowerarm_l, loc=<17.8549 0 0>, rot=<0 0 0 -1>}
 deform_elbow_in_base_l {parent=lowerarm_l, loc=<0.2112 5.645 0.8274>, rot=<0 0 0 -1>}
 deform_elbow_in_l {parent=deform_elbow_in_base_l, loc=<0 0 0>, rot=<0 0 0 1>}
 deform_elbow_out_base_l {parent=lowerarm_l, loc=<0 -7.5129 0>, rot=<0 0 0 -1>}
 deform_elbow_out_l {parent=deform_elbow_out_base_l, loc=<0 0 0>, rot=<0 0 0 1>}
 dyn_forearm_blob_mstr_l {parent=deform_elbow_out_l, loc=<-0.6445 7.5288 9.6483>, rot=<0.0681 -0.7777 0.5915 0.2018>}
 dyn_forearm_blob_01_l {parent=dyn_forearm_blob_mstr_l, loc=<0 0 0>, rot=<0 0 0 1>}
 dyn_forearm_blob_02_l {parent=dyn_forearm_blob_01_l, loc=<2.7434 0 0>, rot=<0 0 0 1>}
 dyn_elb_blob_mstr_l {parent=deform_elbow_out_l, loc=<-5.7557 1.3368 1.0021>, rot=<-0.0038 -0.9954 0.0881 0.0362>}
 dyn_elb_blob_01_l {parent=dyn_elb_blob_mstr_l, loc=<0 0 0>, rot=<0 0 0 1>}
 dyn_elb_blob_02_l {parent=dyn_elb_blob_01_l, loc=<4.2714 0 0>, rot=<0 0 0 1>}
 upperarm_twist_01_l {parent=upperarm_l, loc=<7.9796 0.0843 -0.6195>, rot=<0 0 0 -1>}
 deform_upperarm_bk_base_l {parent=upperarm_twist_01_l, loc=<-6.8983 -13.3826 -1.3258>, rot=<0 0 0 1>}
 deform_upperarm_bk_l {parent=deform_upperarm_bk_base_l, loc=<0 0 0>, rot=<0 0 0 1>}
 deform_upperarm_fr_base_l {parent=upperarm_twist_01_l, loc=<-8.3487 10.7117 -2.1455>, rot=<0 0 0 1>}
 deform_upperarm_fr_l {parent=deform_upperarm_fr_base_l, loc=<0 0 0>, rot=<0 0 0 1>}
 deform_upperarm_up_base_l {parent=upperarm_twist_01_l, loc=<-11.2012 -1.2234 8.7784>, rot=<0 0 0 1>}
 deform_upperarm_up_l {parent=deform_upperarm_up_base_l, loc=<0 0 0>, rot=<0 0 0 1>}
 dyn_shld_blob_mstr_l {parent=upperarm_twist_01_l, loc=<-6.1323 -0.9312 12.8384>, rot=<0.3341 -0.7214 0.584 0.1639>}
 dyn_shld_blob_01_l {parent=dyn_shld_blob_mstr_l, loc=<0 0 0>, rot=<0 0 0 1>}
 dyn_shld_blob_02_l {parent=dyn_shld_blob_01_l, loc=<3.5011 0 0>, rot=<0 0 0 1>}
 upperarm_twist_02_l {parent=upperarm_l, loc=<19.5059 0.0843 -0.6195>, rot=<0 0 0 -1>}
 deform_bicep_base_l {parent=upperarm_twist_02_l, loc=<0 8.0494 0>, rot=<0 0 0 1>}
 deform_bicep_l {parent=deform_bicep_base_l, loc=<0 0 0>, rot=<0 0 0 1>}
 clavicle_r {parent=spine_05, loc=<2.5855 -0.2866 4.3777>, rot=<0.7289 0.0465 -0.6829 0.009>}
 upperarm_r {parent=clavicle_r, loc=<-18.2264 -0.1342 0.001>, rot=<0.0253 -0.2693 0.0013 -0.9627>}
 lowerarm_r {parent=upperarm_r, loc=<-29.0071 -0.1686 1.239>, rot=<0.0391 -0.0099 -0.2453 -0.9686>}
 hand_r {parent=lowerarm_r, loc=<-25.8607 -0.222 -1.0274>, rot=<-0.7193 -0.0988 0.1193 -0.6773>}
 thumb_01_r {parent=hand_r, loc=<-1.3118 0.3191 2.459>, rot=<0.6546 -0.4042 0.0664 -0.6354>}
 thumb_02_r {parent=thumb_01_r, loc=<-5.9336 0.0448 -0.0401>, rot=<-0.0037 0.0213 0.298 -0.9543>}
 thumb_03_r {parent=thumb_02_r, loc=<-2.2223 -0.3682 0.0907>, rot=<0.0118 -0.013 0.0066 -0.9998>}
 index_metacarpal_r {parent=hand_r, loc=<-1.6535 -0.0183 2.96>, rot=<0.0158 -0.0124 -0.0173 -0.9996>}
 index_01_r {parent=index_metacarpal_r, loc=<-8.3008 -0.0357 0.0524>, rot=<-0.0044 0.0485 0.0924 -0.9945>}
 index_02_r {parent=index_01_r, loc=<-3.3587 0.3004 -0.0452>, rot=<-0.002 0.0027 0.3225 -0.9466>}
 index_03_r {parent=index_02_r, loc=<-2.5578 0.123 0.0252>, rot=<-0.002 0.0067 0.2558 -0.9667>}
 middle_metacarpal_r {parent=hand_r, loc=<-2.1856 0.1725 0.5984>, rot=<0.006 0.0168 -0.0329 -0.9993>}
 middle_01_r {parent=middle_metacarpal_r, loc=<-7.3275 0.2447 -0.0129>, rot=<-0.0208 0.0042 0.0858 -0.9961>}
 middle_02_r {parent=middle_01_r, loc=<-4.0123 -0.1401 -0.1175>, rot=<-0.0218 0.0258 0.3512 -0.9357>}
 middle_03_r {parent=middle_02_r, loc=<-2.6077 -0.0534 -0.1373>, rot=<0.0082 0.0188 0.1708 -0.9851>}
 ring_metacarpal_r {parent=hand_r, loc=<-2.2074 0.3625 -1.3638>, rot=<-0.0284 0.0722 -0.0458 -0.9959>}
 ring_01_r {parent=ring_metacarpal_r, loc=<-7.0295 0.6197 0.1446>, rot=<-0.0489 -0.0252 0.1892 -0.9804>}
 ring_02_r {parent=ring_01_r, loc=<-3.6174 -0.4647 0.1027>, rot=<0 0 0.2806 -0.9598>}
 ring_03_r {parent=ring_02_r, loc=<-2.4152 -0.114 -0.0199>, rot=<0 0 0.1189 -0.9929>}
 pinky_metacarpal_r {parent=hand_r, loc=<-1.7616 0.9753 -3.3814>, rot=<-0.0397 0.0814 -0.069 -0.9935>}
 pinky_01_r {parent=pinky_metacarpal_r, loc=<-7.1709 0.4831 0.1094>, rot=<-0.0654 -0.0035 0.2242 -0.9723>}
 pinky_02_r {parent=pinky_01_r, loc=<-2.8274 -0.0893 0.1043>, rot=<0.0037 0.0011 0.2854 -0.9584>}
 pinky_03_r {parent=pinky_02_r, loc=<-2.2046 -0.0307 0.0078>, rot=<-0.0002 0.0098 0.1445 -0.9895>}
 deform_wrist_dn_base_r {parent=hand_r, loc=<0 3.4172 0>, rot=<0 0 0 -1>}
 deform_wrist_dn_r {parent=deform_wrist_dn_base_r, loc=<0 0 0>, rot=<0 0 0 -1>}
 deform_wrist_up_base_r {parent=hand_r, loc=<-0.0001 -2.7804 0>, rot=<0 0 0 -1>}
 deform_wrist_up_r {parent=deform_wrist_up_base_r, loc=<0 0 0>, rot=<0 0 0 -1>}
 lowerarm_twist_02_r {parent=lowerarm_r, loc=<-9.6199 0 0>, rot=<0 0 0 -1>}
 lowerarm_twist_01_r {parent=lowerarm_r, loc=<-17.8549 0 0>, rot=<0 0 0 -1>}
 deform_elbow_in_base_r {parent=lowerarm_r, loc=<-0.2114 -5.6449 -0.827>, rot=<0 0 0 -1>}
 deform_elbow_in_r {parent=deform_elbow_in_base_r, loc=<0 0 0>, rot=<0 0 0 1>}
 deform_elbow_out_base_r {parent=lowerarm_r, loc=<-0.0001 7.5129 0.0002>, rot=<0 0 0 -1>}
 deform_elbow_out_r {parent=deform_elbow_out_base_r, loc=<0 0 0>, rot=<0 0 0 1>}
 dyn_forearm_blob_mstr_r {parent=deform_elbow_out_r, loc=<0.6322 -7.5201 -9.6336>, rot=<0.0684 -0.7777 0.5915 0.2015>}
 dyn_forearm_blob_01_r {parent=dyn_forearm_blob_mstr_r, loc=<0 0 0>, rot=<0 0 0 1>}
 dyn_forearm_blob_02_r {parent=dyn_forearm_blob_01_r, loc=<-2.7428 -0.0003 -0.0002>, rot=<0 0 0 1>}
 dyn_elb_blob_mstr_r {parent=deform_elbow_out_r, loc=<5.7557 -1.3369 -1.0019>, rot=<-0.0038 -0.9954 0.0881 0.0362>}
 dyn_elb_blob_01_r {parent=dyn_elb_blob_mstr_r, loc=<0 0 0>, rot=<0 0 0 1>}
 dyn_elb_blob_02_r {parent=dyn_elb_blob_01_r, loc=<-4.2717 0 0.0003>, rot=<0 0 0 1>}
 upperarm_twist_01_r {parent=upperarm_r, loc=<-7.9796 -0.0843 0.6195>, rot=<0 0 0 -1>}
 deform_upperarm_bk_base_r {parent=upperarm_twist_01_r, loc=<6.8985 13.3826 1.3257>, rot=<0 0 0 1>}
 deform_upperarm_bk_r {parent=deform_upperarm_bk_base_r, loc=<0 0 0>, rot=<0 0 0 1>}
 deform_upperarm_fr_base_r {parent=upperarm_twist_01_r, loc=<8.3485 -10.7117 2.1456>, rot=<0 0 0 1>}
 deform_upperarm_fr_r {parent=deform_upperarm_fr_base_r, loc=<0 0 0>, rot=<0 0 0 1>}
 deform_upperarm_up_base_r {parent=upperarm_twist_01_r, loc=<11.2015 1.2234 -8.7788>, rot=<0 0 0 1>}
 deform_upperarm_up_r {parent=deform_upperarm_up_base_r, loc=<0 0 0>, rot=<0 0 0 1>}
 dyn_shld_blob_mstr_r {parent=upperarm_twist_01_r, loc=<6.1322 0.9312 -12.8382>, rot=<0.3341 -0.7214 0.584 0.1639>}
 dyn_shld_blob_01_r {parent=dyn_shld_blob_mstr_r, loc=<0 0 0>, rot=<0 0 0 1>}
 dyn_shld_blob_02_r {parent=dyn_shld_blob_01_r, loc=<-3.5015 0.0001 0.0001>, rot=<0 0 0 1>}
 upperarm_twist_02_r {parent=upperarm_r, loc=<-19.5059 -0.0843 0.6195>, rot=<0 0 0 -1>}
 deform_bicep_base_r {parent=upperarm_twist_02_r, loc=<0.0002 -8.0494 -0.0003>, rot=<0 0 0 1>}
 deform_bicep_r {parent=deform_bicep_base_r, loc=<0 0 0>, rot=<0 0 0 1>}
 deform_pec_aim_base_l {parent=spine_04, loc=<5.0232 13.6719 -2.6508>, rot=<0.0672 -0.5201 0.3228 -0.7879>}
 deform_pec_aim_l {parent=deform_pec_aim_base_l, loc=<0 0 0>, rot=<0 0 0 1>}
 deform_pec_base_l {parent=deform_pec_aim_l, loc=<16.3279 0 0>, rot=<0 0 0 1>}
 deform_pec_l {parent=deform_pec_base_l, loc=<0 0 0>, rot=<0 0 0 1>}
 deform_pec_aim_base_r {parent=spine_04, loc=<5.023 13.6718 2.6508>, rot=<-0.5201 -0.0672 0.7879 0.3228>}
 deform_pec_aim_r {parent=deform_pec_aim_base_r, loc=<0 0 0>, rot=<0 0 0 1>}
 deform_pec_base_r {parent=deform_pec_aim_r, loc=<-16.3279 0 0>, rot=<0 0 0 -1>}
 deform_pec_r {parent=deform_pec_base_r, loc=<0 0 0.0002>, rot=<0 0 0 -1>}
 deform_lat_l_base {parent=spine_03, loc=<2.5566 -10.3716 -17.6918>, rot=<0 0 0 1>}
 deform_lat_l {parent=deform_lat_l_base, loc=<0 0 0>, rot=<0 0 0 1>}
 deform_lat_r_base {parent=spine_03, loc=<2.5568 -10.3716 17.6918>, rot=<0 0 1 0>}
 deform_lat_r {parent=deform_lat_r_base, loc=<0 0 0>, rot=<0 0 0 -1>}
 thigh_l {parent=pelvis, loc=<-1.8716 -0.0109 -11.9936>, rot=<0.0012 0.0572 0.0155 -0.9982>}
 calf_l {parent=thigh_l, loc=<-41.9178 -0.2246 0.0926>, rot=<0.0006 -0.0161 -0.0391 -0.9991>}
 foot_l {parent=calf_l, loc=<-38.7057 1.2438 0.2452>, rot=<0.0975 -0.0474 0.0317 -0.9936>}
 ball_l {parent=foot_l, loc=<-6.1614 13.1469 0.6807>, rot=<0.0001 -0.0001 -0.7063 -0.7079>}
 calf_twist_02_l {parent=calf_l, loc=<-15.9268 -0.0967 0.0397>, rot=<0.0237 0.0095 -0.0035 -0.9997>}
 dyn_blob_calf_mstr {parent=calf_twist_02_l, loc=<8.9573 -10.5062 -1.0178>, rot=<0.9931 -0.0807 -0.0526 0.0664>}
 dyn_blob_calf_01_l {parent=dyn_blob_calf_mstr, loc=<0 0 0>, rot=<0 0 0 1>}
 dyn_blob_calf_02_l {parent=dyn_blob_calf_01_l, loc=<4.3705 0 0>, rot=<0 0 0 1>}
 calf_twist_01_l {parent=calf_l, loc=<-32.9335 -0.0967 0.0397>, rot=<0.0237 0.0095 -0.0035 -0.9997>}
 deform_knee_in_base_l {parent=calf_l, loc=<1.298 -8.0356 -0.0496>, rot=<0 0 0 -1>}
 deform_knee_in_l {parent=deform_knee_in_base_l, loc=<0 0 0>, rot=<0 0 0 1>}
 deform_knee_out_base_l {parent=calf_l, loc=<0 6.8739 0>, rot=<0 0 0 -1>}
 deform_knee_out_l {parent=deform_knee_out_base_l, loc=<0 0 0>, rot=<0 0 0 1>}
 thigh_twist_01_l {parent=thigh_l, loc=<-21.6303 0 0>, rot=<0.0013 0 0 -1>}
 dyn_armor_mstr_l {parent=thigh_twist_01_l, loc=<29.2237 -0.5676 -4.666>, rot=<-0.0566 0.6541 0.75 0.0805>}
 dyn_armor_l {parent=dyn_armor_mstr_l, loc=<0 0 0>, rot=<0 0 0 1>}
 dyn_thigh_adjust_mstr_l {parent=thigh_twist_01_l, loc=<9.4343 6.2352 9.2713>, rot=<-0.3279 0.2407 -0.5833 -0.7031>}
 dyn_thigh_adjust_l {parent=dyn_thigh_adjust_mstr_l, loc=<0 0 0>, rot=<-0.3279 0.2407 -0.5833 -0.7031>}
 deform_groin_base_l {parent=thigh_l, loc=<-2.1908 8.5661 -1.7442>, rot=<0 0 0 -1>}
 deform_groin_l {parent=deform_groin_base_l, loc=<0 0 0>, rot=<0 0 0 1>}
 thigh_r {parent=pelvis, loc=<-1.8716 -0.0109 11.9936>, rot=<0.0572 -0.0012 0.9982 0.0155>}
 calf_r {parent=thigh_r, loc=<41.9178 0.2246 -0.0926>, rot=<0.0006 -0.0161 -0.0391 -0.9991>}
 foot_r {parent=calf_r, loc=<38.7057 -1.2438 -0.2452>, rot=<0.0975 -0.0474 0.0317 -0.9936>}
 ball_r {parent=foot_r, loc=<6.1614 -13.1469 -0.6807>, rot=<0.0001 -0.0001 -0.7063 -0.7079>}
 calf_twist_02_r {parent=calf_r, loc=<15.9268 0.0967 -0.0397>, rot=<0.0237 0.0095 -0.0035 -0.9997>}
 dyn_blob_calf_mstr_r {parent=calf_twist_02_r, loc=<-8.9573 10.5062 1.0178>, rot=<0.9931 -0.0807 -0.0526 0.0664>}
 dyn_blob_calf_01_r {parent=dyn_blob_calf_mstr_r, loc=<0 0 0>, rot=<0 0 0 1>}
 dyn_blob_calf_02_r {parent=dyn_blob_calf_01_r, loc=<-4.3705 0 0>, rot=<0 0 0 -1>}
 calf_twist_01_r {parent=calf_r, loc=<32.9335 0.0967 -0.0397>, rot=<0.0237 0.0095 -0.0035 -0.9997>}
 deform_knee_in_base_r {parent=calf_r, loc=<-1.2981 8.0356 0.0496>, rot=<0 0 0 -1>}
 deform_knee_in_r {parent=deform_knee_in_base_r, loc=<0 0 0>, rot=<0 0 0 1>}
 deform_knee_out_base_r {parent=calf_r, loc=<0 -6.8739 0>, rot=<0 0 0 -1>}
 deform_knee_out_r {parent=deform_knee_out_base_r, loc=<0 0 0>, rot=<0 0 0 1>}
 thigh_twist_01_r {parent=thigh_r, loc=<21.6303 0 0>, rot=<0.0013 0 0 -1>}
 dyn_armor_mstr_r {parent=thigh_twist_01_r, loc=<-29.2237 0.5676 4.666>, rot=<-0.0566 0.6541 0.75 0.0805>}
 dyn_armor_r {parent=dyn_armor_mstr_r, loc=<0 0 0>, rot=<0 0 0 1>}
 dyn_thigh_adjust_mstr_r {parent=thigh_twist_01_r, loc=<-4.1848 -9.8863 -2.4094>, rot=<0.0394 -0.0415 0.7076 -0.7043>}
 dyn_thigh_adjust_r {parent=dyn_thigh_adjust_mstr_r, loc=<0 0 0>, rot=<0.0394 -0.0415 0.7076 -0.7043>}
 deform_groin_base_r {parent=thigh_r, loc=<2.1908 -8.5661 1.7442>, rot=<0 0 0 -1>}
 deform_groin_r {parent=deform_groin_base_r, loc=<0 0 0>, rot=<0 0 0 -1>}
 deform_glute_base_l {parent=pelvis, loc=<-10.1877 -11.6217 -12.899>, rot=<0.0275 0.0501 -0.4472 -0.8926>}
 deform_glute_l {parent=deform_glute_base_l, loc=<0 0 0>, rot=<0 0 0 1>}
 deform_glute_base_r {parent=pelvis, loc=<-10.1877 -11.6218 12.899>, rot=<0.0501 -0.0275 0.8926 -0.4472>}
 deform_glute_r {parent=deform_glute_base_r, loc=<0 0 0>, rot=<0 0 0 1>}
 dyn_armor_fr_mstr {parent=pelvis, loc=<7.1643 12.4038 0>, rot=<0 0 0.9999 0.0135>}
 dyn_armor_fr_01 {parent=dyn_armor_fr_mstr, loc=<0 0 0>, rot=<0 0 0.0034 -1>}
 dyn_armor_fr_02 {parent=dyn_armor_fr_01, loc=<17.8782 0 0>, rot=<0 0 0.0061 -1>}
 dyn_buckle_const {parent=pelvis, loc=<6.2024 12.0108 0.0765>, rot=<0 0 -0.6943 -0.7197>}
 dyn_buckle_mstr {parent=dyn_buckle_const, loc=<0 0 0>, rot=<0 0 0 1>}
 dyn_buckle {parent=dyn_buckle_mstr, loc=<0 0 0>, rot=<0 0 0 1>}
 dyn_shell_01_mstr_l {parent=pelvis, loc=<9.0716 0.5105 -17.6171>, rot=<-0.0914 0.6411 0.7444 0.1628>}
 dyn_shell_01_l {parent=dyn_shell_01_mstr_l, loc=<0 0 0>, rot=<0 0 0 1>}
 dyn_shell_02_mstr_l {parent=pelvis, loc=<8.7146 -8.2324 -15.5405>, rot=<0.1508 -0.8404 -0.5117 -0.0955>}
 dyn_shell_02_l {parent=dyn_shell_02_mstr_l, loc=<0 0 0>, rot=<0 0 0 1>}
 dyn_shell_01_mstr_r {parent=pelvis, loc=<8.715 0.3657 17.218>, rot=<-0.6411 -0.0914 0.1628 -0.7444>}
 dyn_shell_01_r {parent=dyn_shell_01_mstr_r, loc=<0 0 0>, rot=<0 0 0 1>}
 dyn_shell_02_mstr_r {parent=pelvis, loc=<8.715 -8.0521 15.1761>, rot=<-0.8404 -0.1508 0.0955 -0.5117>}
 dyn_shell_02_r {parent=dyn_shell_02_mstr_r, loc=<0 0 0>, rot=<0 0 0 1>}
 dyn_armor_bk_mstr {parent=pelvis, loc=<7.2051 -13.9803 0>, rot=<0.0388 -0.9992 0 0>}
 dyn_armor_bk_01 {parent=dyn_armor_bk_mstr, loc=<0 0 0>, rot=<0 0 0 1>}
 dyn_armor_bk_02 {parent=dyn_armor_bk_01, loc=<27.1597 -0.2796 0>, rot=<0 0 -0.0039 -1>}
 dyn_conch_mstr {parent=pelvis, loc=<6.3265 -12.9654 -8.1361>, rot=<0.7985 -0.5964 -0.0469 -0.0678>}
 dyn_conch {parent=dyn_conch_mstr, loc=<0 0 0>, rot=<0 0 0 1>}
 dyn_conch_ring {parent=dyn_conch, loc=<3.0816 9.3507 1.361>, rot=<-0.0847 0.2782 -0.6602 -0.6925>}
 dyn_conch_eye_harmonics_mstr_r {parent=dyn_conch, loc=<4.319 6.8634 1.564>, rot=<-0.2439 0.122 -0.4939 -0.8257>}
 dyn_conch_eye_harmonics_r {parent=dyn_conch_eye_harmonics_mstr_r, loc=<0 0 0>, rot=<-0.2439 0.122 -0.4939 -0.8257>}
 dyn_conch_eye_root_r {parent=dyn_conch_eye_harmonics_r, loc=<0 0 0>, rot=<0 0 0 1>}
 dyn_conch_eye_r {parent=dyn_conch_eye_root_r, loc=<0 0 0>, rot=<0 0 0 1>}
 dyn_conch_eye_harmonics_mstr_l {parent=dyn_conch, loc=<3.7296 8.9985 -1.5728>, rot=<-0.2571 0.0646 -0.4789 -0.8369>}
 dyn_conch_eye_harmonics_l {parent=dyn_conch_eye_harmonics_mstr_l, loc=<0 0 0>, rot=<-0.2571 0.0646 -0.4789 -0.8369>}
 dyn_conch_eye_root_l {parent=dyn_conch_eye_harmonics_l, loc=<0 0 0>, rot=<0 0 0 1>}
 dyn_conch_eye_l {parent=dyn_conch_eye_root_l, loc=<0 0 0>, rot=<0 0 0 1>}
 root {parent=root, type=Socket, loc=<0 0 0>, rot=<0 0 0 1>}
 SPR_Thigh_R {type=Curve, loc=<0 0 0>, rot=<0 0 0 1>}
]
variables [
 Spine_Adjust {type=double, default=0}
 Thigh_L_Adjust {type=double, default=0}
 Thigh_L_Strafe_Adjust {type=double, default=0}
 Thigh_R_Adjust {type=double, default=0}
 Thigh_R_Strafe_Adjust {type=double, default=0}
]
