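# Physics asset for the TideKing Regal large male player body, covering dangling front and back armor, a conch and four side shells.
# Each dynamic chain hangs from a tiny master-bone capsule through limited constraints: armor swings mostly on one axis, shells only slightly.
# Capsules on pelvis, lower spine, thighs and calves give the body shapes; the back armor tip uses a box.
# Leg constraints leave rotation free, and only the right thigh links to the pelvis.
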PhysicsAsset /BRCosmetics/Characters/Player/Male/Large/Bodies/M_LRG_TideKing_Regal/Meshes/M_LRG_TideKing_Regal_Physics
bodies {
 calf_l [{
   shapeType=Capsule
   length=41.1983
   radius=8.4373
   shapeName=calf_l_capsule
   center=<-21.7543 -3.312 0>
   rotation=<rot -90 -179.5524 177.5423>
  }]
 calf_r [{shapeType=Capsule, length=41.1983, radius=8.4373, shapeName=calf_r_capsule, center=<21.13 4.0934 0>, rotation=<rot -90 -160.2328 -21.7773>}]
 dyn_armor_bk_01 [{
   shapeType=Capsule
   length=20.9958
   radius=9.1424
   shapeName=dyn_armor_bk_01_capsule
   center=<22.063 -4.9078 0>
   rotation=<rot -90 180 177.0377>
  }]
 dyn_armor_bk_02 [{shapeType=Box, extentX=23.5086, extentY=7.5, extentZ=47.1585, shapeName=dyn_armor_bk_02_box, center=<21.1701 -1.9981 0>}]
 dyn_armor_bk_mstr [{shapeType=Capsule, length=0.505, radius=0.505, shapeName=dyn_armor_bk_mstr_capsule, rotation=<rot -90 0 0>}]
 dyn_armor_fr_01 [{shapeType=Capsule, length=7.6409, radius=5.8112, shapeName=dyn_armor_fr_01_capsule, center=<5.6401 -6.6126 0>, rotation=<rot 0 180 180>}]
 dyn_armor_fr_02 [{shapeType=Capsule, length=14.1682, radius=8.9463, shapeName=dyn_armor_fr_02_capsule, center=<4.9786 -3.1774 0>, rotation=<rot -90 0 0>}]
 dyn_armor_fr_mstr [{shapeType=Capsule, length=0.505, radius=0.505, shapeName=dyn_armor_fr_mstr_capsule, rotation=<rot -90 0 0>}]
 dyn_conch [{shapeType=Capsule, length=6.0267, radius=5.8289, shapeName=dyn_conch_capsule, center=<5.007 1.9741 0.4962>, rotation=<rot -90 0 0>}]
 dyn_conch_mstr [{shapeType=Capsule, length=0.505, radius=0.505, shapeName=dyn_conch_mstr_capsule, rotation=<rot -90 0 0>}]
 dyn_shell_01_l [{shapeType=Capsule, length=5.4596, radius=4.422, shapeName=dyn_shell_01_l_capsule, center=<1.5583 -2.6765 -0.2992>, rotation=<rot 0 0 90>}]
 dyn_shell_01_mstr_l [{shapeType=Capsule, length=0.505, radius=0.505, shapeName=dyn_shell_01_mstr_l_capsule, rotation=<rot 0 -90 0>}]
 dyn_shell_01_mstr_r [{shapeType=Capsule, length=0.505, radius=0.505, shapeName=dyn_shell_01_mstr_r_capsule, rotation=<rot 0 -90 0>}]
 dyn_shell_01_r [{shapeType=Capsule, length=6.2597, radius=3.622, shapeName=dyn_shell_01_r_capsule, center=<-1.121 2.7242 0.1952>, rotation=<rot 0 0 90>}]
 dyn_shell_02_l [{shapeType=Capsule, length=6.2544, radius=3.6406, shapeName=dyn_shell_02_l_capsule, center=<2.5013 -2.9693 0.0542>, rotation=<rot 0 0 90>}]
 dyn_shell_02_mstr_l [{shapeType=Capsule, length=0.505, radius=0.505, shapeName=dyn_shell_02_mstr_l_capsule, rotation=<rot 0 -90 0>}]
 dyn_shell_02_mstr_r [{shapeType=Capsule, length=0.505, radius=0.505, shapeName=dyn_shell_02_mstr_r_capsule, rotation=<rot 0 -90 0>}]
 dyn_shell_02_r [{shapeType=Capsule, length=5.4544, radius=4.4407, shapeName=dyn_shell_02_r_capsule, center=<-2.082 2.5453 -0.0609>, rotation=<rot 0 0 90>}]
 pelvis [{shapeType=Capsule, length=14.9549, radius=8.5704, shapeName=pelvis_capsule, center=<-0.4454 -4.0743 0>}]
 spine_01 [{shapeType=Capsule, length=13.9591, radius=10.9749, shapeName=spine_01_capsule}]
 spine_02 [{shapeType=Capsule, length=9.5333, radius=11.8496, shapeName=spine_02_capsule}]
 thigh_l [{
   shapeType=Capsule
   length=32.8909
   radius=10.5611
   shapeName=thigh_l_capsule
   center=<-14.6102 0 2.8883>
   rotation=<rot -88.2354 -0.0012 3.6777>
  }]
 thigh_r [{
   shapeType=Capsule
   length=34.1604
   radius=10.5611
   shapeName=thigh_r_capsule
   center=<14.6102 0 -2.6112>
   rotation=<rot -88.2354 -0.0011 3.6777>
  }]
}
constraints [
 {
  bone1Name=thigh_r
  bone2Name=pelvis
  swing1LimitDegrees=45
  swing1Motion=Free
  swing2LimitDegrees=45
  swing2Motion=Free
  twistLimitDegrees=45
  twistMotion=Free
 }
 {bone1Name=dyn_armor_bk_01, bone2Name=dyn_armor_bk_mstr, swing1LimitDegrees=25, swing2LimitDegrees=2, twistLimitDegrees=2}
 {bone1Name=dyn_armor_bk_02, bone2Name=dyn_armor_bk_01, swing1LimitDegrees=12, swing2LimitDegrees=2, twistLimitDegrees=2}
 {
  bone1Name=calf_r
  bone2Name=thigh_r
  swing1LimitDegrees=45
  swing1Motion=Free
  swing2LimitDegrees=45
  swing2Motion=Free
  twistLimitDegrees=45
  twistMotion=Free
 }
 {
  bone1Name=calf_l
  bone2Name=thigh_l
  swing1LimitDegrees=45
  swing1Motion=Free
  swing2LimitDegrees=45
  swing2Motion=Free
  twistLimitDegrees=45
  twistMotion=Free
 }
 {bone1Name=dyn_conch, bone2Name=dyn_conch_mstr, swing1LimitDegrees=20, swing2LimitDegrees=20, twistLimitDegrees=30}
 {bone1Name=dyn_armor_fr_01, bone2Name=dyn_armor_fr_mstr, swing1LimitDegrees=25, swing2LimitDegrees=2, twistLimitDegrees=2}
 {bone1Name=dyn_armor_fr_02, bone2Name=dyn_armor_fr_01, swing1LimitDegrees=10, swing2LimitDegrees=2, twistLimitDegrees=2}
 {bone1Name=dyn_shell_01_l, bone2Name=dyn_shell_01_mstr_l, swing1LimitDegrees=1, swing2LimitDegrees=8, twistLimitDegrees=8}
 {bone1Name=dyn_shell_01_r, bone2Name=dyn_shell_01_mstr_r, swing1LimitDegrees=1, swing2LimitDegrees=8, twistLimitDegrees=8}
 {bone1Name=dyn_shell_02_l, bone2Name=dyn_shell_02_mstr_l, swing1LimitDegrees=1, swing2LimitDegrees=8, twistLimitDegrees=8}
 {bone1Name=dyn_shell_02_r, bone2Name=dyn_shell_02_mstr_r, swing1LimitDegrees=1, swing2LimitDegrees=8, twistLimitDegrees=8}
]
