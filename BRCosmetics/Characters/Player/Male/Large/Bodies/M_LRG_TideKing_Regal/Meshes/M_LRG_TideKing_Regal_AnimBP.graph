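# Post-copy secondary-motion AnimBP for the large male Tide King Regal cosmetic body.
# It copies the attached parent's pose through a Control Rig, then chains RigidBody (body, arms), curve-driven front-armor bone scaling and AnimDynamics (conch ring, calf blobs).
# Each update recomputes gravity via a deprecated helper and halves physics alpha while skydiving.
# Arm simulations use a separate Blob_Gravity, and front-armor scale blends by the SPR_Thigh_R curve.
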AnimBlueprint /BRCosmetics/Characters/Player/Male/Large/Bodies/M_LRG_TideKing_Regal/Meshes/M_LRG_TideKing_Regal_AnimBP
target_skeleton=&/M_LRG_TideKing_Regal_Skeleton
graphs [
 AnimGraph {nodes [
   n1 = Control Rig(Source=Copy Pose From Mesh() {bUseAttachedParent=true}) {
    alphaBoolBlend {blendOption=Linear}
    controlRigAssetReference {blueprintRigClass=&/M_LRG_TideKing_Regal_CtrlRig.M_LRG_TideKing_Regal_CtrlRig_C}
   }
   n2 = Get Skydive_Alpha() -> (Skydive_Alpha:double)
   n3 = Get Gravity Override()
   n4 = RigidBody!(ComponentPose=Local To Component(LocalPose=@n1), Alpha=@n2, OverrideWorldGravity:Vector=@n3, ExternalForce) -> (Pose) {
    bClampLinearTranslationLimitToRefPose=true
    bOverrideWorldGravity=true
    overridePhysicsAsset=&/M_LRG_TideKing_Regal_Physics
    simulationSpace=BaseBoneSpace
    alphaBoolBlend {blendOption=Linear}
    baseBoneRef {boneName=pelvis}
    componentAppliedLinearAccClamp=<5000 5000 5000>
    componentLinearAccScale=<0.6 0.6 0.6>
    componentLinearVelScale=<0.6 0.6 0.6>
    simSpaceSettings {maxAngularAcceleration=300, maxAngularVelocity=10, maxLinearAcceleration=150, maxLinearVelocity=300, worldAlpha=0.5}
   }
   n5 = Get Blob_Gravity() -> (Blob_Gravity:Vector)
   n6 = RigidBody(ComponentPose=@n4, Alpha=@n2, OverrideWorldGravity=@n5) {
    bClampLinearTranslationLimitToRefPose=true
    bOverrideWorldGravity=true
    overridePhysicsAsset=&/M_LRG_TideKing_Regal_Arm_R_Physics
    simulationSpace=BaseBoneSpace
    alphaBoolBlend {blendOption=Linear}
    baseBoneRef {boneName=upperarm_r}
    componentAppliedLinearAccClamp=<5000 5000 5000>
    componentLinearAccScale=<0.4 0.4 0.5>
    componentLinearVelScale=<0.4 0.4 0.5>
    simSpaceSettings {maxAngularAcceleration=6, maxAngularVelocity=10, maxLinearAcceleration=100, maxLinearVelocity=200, velocityScaleZ=0.5, worldAlpha=0.5}
   }
   n7 = RigidBody(ComponentPose=@n6, Alpha=@n2, OverrideWorldGravity=@n5) {
    bClampLinearTranslationLimitToRefPose=true
    bOverrideWorldGravity=true
    overridePhysicsAsset=&/M_LRG_TideKing_Regal_Arm_L_Physics
    simulationSpace=BaseBoneSpace
    alphaBoolBlend {blendOption=Linear}
    baseBoneRef {boneName=upperarm_l}
    componentAppliedLinearAccClamp=<5000 5000 5000>
    componentLinearAccScale=<0.4 0.4 0.5>
    componentLinearVelScale=<0.4 0.4 0.5>
    simSpaceSettings {maxAngularAcceleration=6, maxAngularVelocity=10, maxLinearAcceleration=100, maxLinearVelocity=200, velocityScaleZ=0.5, worldAlpha=0.5}
   }
   n8 = Get Armor_Fr_Scale_Alpha() -> (Armor_Fr_Scale_Alpha:double)
   n9 = "Transform (Modify) Bone - Bone: dyn_armor_fr_01"(ComponentPose=@n7, Alpha=@n8, Translation="0,0,0", Rotation="0,0,0", Scale="1,1,0.6") {scaleMode=BMM_Replace, scaleSpace=BCS_BoneSpace, alphaBoolBlend {blendOption=Linear}, boneToModify {boneName=dyn_armor_fr_01}, scale {z=0.6}}
   n10 = "Transform (Modify) Bone - Bone: dyn_armor_fr_02"(ComponentPose=@n9, Alpha=@n8, Translation="0,0,0", Rotation="0,0,0", Scale="1,1,1.75") {
    scaleMode=BMM_Replace
    scaleSpace=BCS_BoneSpace
    alphaBoolBlend {blendOption=Linear}
    boneToModify {boneName=dyn_armor_fr_02}
    scale {z=1.75}
   }
   n11 = Anim Dynamics - Bone: dyn_conch_ring!(ComponentPose=@n10, Alpha=@n2, GravityOverride:Vector=@n3, ExternalForce="0,0,0") -> (Pose) {
    bUseGravityOverride=true
    simulationSpace=BoneRelative
    alphaBoolBlend {blendOption=Linear}
    boundBone {boneName=dyn_conch_ring}
    componentAppliedLinearAccClamp=<10000 10000 10000>
    componentLinearAccScale=<1 1 1>
    componentLinearVelScale=<1 1 1>
    physicsBodyDefinitions [{
      boundBone {boneName=dyn_conch_ring}
      boxExtents=<8 8 8>
      constraintSetup {angularLimitsMax=<0 0 40>, angularLimitsMin=<0 0 -40>, angularTarget=<0 0 0>}
      localJointOffset=<0.7 0 0>
     }]
    relativeSpaceBone {boneName=root}
   }
   n12 = Anim Dynamics - Bone: dyn_blob_calf_01_l!(ComponentPose=@n11, Alpha=@n2, GravityOverride:Vector=@n3, ExternalForce="0,0,0") -> (Pose) {
    angularDampingOverride=0.95
    angularSpringConstant=7500
    bAngularSpring=true
    bOverrideAngularDamping=true
    bUseGravityOverride=true
    linearDampingOverride=0.9999
    linearSpringConstant=335
    simulationSpace=BoneRelative
    alphaBoolBlend {blendOption=Linear}
    boundBone {boneName=dyn_blob_calf_01_l}
    componentAppliedLinearAccClamp=<3000 3000 3000>
    componentLinearAccScale=<0.6 0.6 0.6>
    componentLinearVelScale=<0.6 0.6 0.6>
    physicsBodyDefinitions [{
      boundBone {boneName=dyn_blob_calf_01_l}
      boxExtents=<30 30 30>
      constraintSetup {angularLimitsMax=<2 16 16>, angularLimitsMin=<-2 -16 -16>, angularTarget=<1 0 0>}
      localJointOffset=<0.1 0 0>
     }]
    relativeSpaceBone {boneName=root}
   }
   n13 = Anim Dynamics - Bone: dyn_blob_calf_02_l!(ComponentPose=@n12, Alpha=@n2, GravityOverride:Vector=@n3, ExternalForce="0,0,0") -> (Pose) {
    angularDampingOverride=0.95
    angularSpringConstant=7500
    bAngularSpring=true
    bOverrideAngularDamping=true
    bUseGravityOverride=true
    linearDampingOverride=0.9999
    linearSpringConstant=335
    simulationSpace=BoneRelative
    alphaBoolBlend {blendOption=Linear}
    boundBone {boneName=dyn_blob_calf_02_l}
    componentAppliedLinearAccClamp=<3000 3000 3000>
    componentLinearAccScale=<0.6 0.6 0.6>
    componentLinearVelScale=<0.6 0.6 0.6>
    physicsBodyDefinitions [{
      boundBone {boneName=dyn_blob_calf_02_l}
      boxExtents=<30 30 30>
      constraintSetup {angularLimitsMax=<2 16 16>, angularLimitsMin=<-2 -16 -16>, angularTarget=<1 0 0>}
      localJointOffset=<0.1 0 0>
     }]
    relativeSpaceBone {boneName=root}
   }
   n14 = Anim Dynamics - Bone: dyn_blob_calf_01_r!(ComponentPose=@n13, Alpha=@n2, GravityOverride:Vector=@n3, ExternalForce="0,0,0") -> (Pose) {
    angularDampingOverride=0.99
    angularSpringConstant=7500
    bAngularSpring=true
    bOverrideAngularDamping=true
    bUseGravityOverride=true
    linearDampingOverride=0.9999
    linearSpringConstant=335
    simulationSpace=BoneRelative
    alphaBoolBlend {blendOption=Linear}
    boundBone {boneName=dyn_blob_calf_01_r}
    componentAppliedLinearAccClamp=<3000 3000 3000>
    componentLinearAccScale=<0.7 0.7 0.7>
    componentLinearVelScale=<0.7 0.7 0.7>
    physicsBodyDefinitions [{
      boundBone {boneName=dyn_blob_calf_01_r}
      boxExtents=<25 25 25>
      constraintSetup {angularLimitsMax=<2 16 16>, angularLimitsMin=<-2 -16 -16>, angularTarget=<1 0 0>}
      localJointOffset=<0.1 0 0>
     }]
    relativeSpaceBone {boneName=root}
   }
   n15 = Anim Dynamics - Bone: dyn_blob_calf_02_r!(ComponentPose=@n14, Alpha=@n2, GravityOverride:Vector=@n3, ExternalForce="0,0,0") -> (Pose) {
    angularDampingOverride=0.99
    angularSpringConstant=7500
    bAngularSpring=true
    bOverrideAngularDamping=true
    bUseGravityOverride=true
    linearDampingOverride=0.9999
    linearSpringConstant=335
    simulationSpace=BoneRelative
    alphaBoolBlend {blendOption=Linear}
    boundBone {boneName=dyn_blob_calf_02_r}
    componentAppliedLinearAccClamp=<3000 3000 3000>
    componentLinearAccScale=<0.7 0.7 0.7>
    componentLinearVelScale=<0.7 0.7 0.7>
    physicsBodyDefinitions [{
      boundBone {boneName=dyn_blob_calf_02_r}
      boxExtents=<25 25 25>
      constraintSetup {angularLimitsMax=<2 16 16>, angularLimitsMin=<-2 -16 -16>, angularTarget=<1 0 0>}
      localJointOffset=<0.1 0 0>
     }]
    relativeSpaceBone {boneName=root}
   }
   Output Pose(Result=Component To Local(ComponentPose=@n15)) {layerGroup=None}
  ]}
 EventGraph {nodes [
   n1 = Event BlueprintUpdateAnimation()
   > n2 = Sequence!(execute) -> (then_0, then_1, then_2:Exec)
   > n3 = Gravity Override State Function DEPRECATED(CharacterPartAnimInstance=Self-Reference(), GameTimeInSeconds=@n1.DeltaTimeX, GravityOverrideParams=Get Gravity Override Params())
   > Set Gravity Override(Gravity Override=@n3)
   n4 = Branch(execute=@n2.then_1, Condition=Get bIsSkyDiving())
   > Set Skydive_Alpha(Skydive_Alpha:double=0.5) -> (Output_Get:double)
   Set Skydive_Alpha(execute=@n4.else, Skydive_Alpha=1)
   Set Armor_Fr_Scale_Alpha(execute=@n2.then_2, Armor_Fr_Scale_Alpha:double=GetCurveValue(CurveName=SPR_Thigh_R)) -> (Output_Get:double)
  ]}
]
variables [
 Armor_Fr_Scale_Alpha {type=number}
 Blob_Gravity {type=Vector}
 Gravity Override {type=Vector}
 Gravity Override Params {type=GravityOverrideParamsStruct}
 Skydive_Alpha {type=number}
]
