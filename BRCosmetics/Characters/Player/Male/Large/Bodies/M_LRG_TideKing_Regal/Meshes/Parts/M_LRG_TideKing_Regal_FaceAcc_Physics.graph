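# Physics asset for the face accessory mesh of the large male TideKing Regal player character, set up for a necklace bone.
# Capsule bodies sit on dyn_necklace and spine_05, with the smaller necklace capsule offset and rotated from its bone.
# A single constraint links the necklace to the spine, tightly limiting twist and one swing axis while allowing a wider swing on the other, so motion is mostly in one plane.
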PhysicsAsset /BRCosmetics/Characters/Player/Male/Large/Bodies/M_LRG_TideKing_Regal/Meshes/Parts/M_LRG_TideKing_Regal_FaceAcc_Physics
bodies {
 dyn_necklace [{shapeType=Capsule, length=9.6789, radius=3.3879, shapeName=dyn_necklace_capsule, center=<7.2877 -3.401 0>, rotation=<rot -90 90 -80.4515>}]
 spine_05 [{shapeType=Capsule, length=19.7158, radius=12.2838, shapeName=spine_05_capsule}]
}
constraints [{bone1Name=dyn_necklace, bone2Name=spine_05, swing1LimitDegrees=2, swing2LimitDegrees=20, twistLimitDegrees=2}]
